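# Physics asset for the right-leg dynamic shorts of the Majesty Taco medium male player cosmetic body.
# Capsule bodies cover the thigh, calf and front, back, inner and outer shorts bones.
# Each flap links with tight swing and twist limits to a small master body, with the back flap as a two-bone chain.
# The master bodies attach to the upper thigh twist bone with free angular motion.
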PhysicsAsset /BRCosmetics/Characters/Player/Male/Medium/Bodies/M_MED_Majesty_Taco/Meshes/M_MED_Majesty_Taco_dyn_pants_r
bodies {
 calf_r [{shapeType=Capsule, length=29.5, radius=9.75, center=<14.721 1.9012 -0.4492>, rotation=<rot 88.4112 0.1456 -4.552>}]
 dyn_shorts_bk_01_r [{shapeType=Capsule, length=19.5, radius=4.75, center=<5 0 3>, rotation=<rot 0 0 90>}]
 dyn_shorts_bk_02_r [{shapeType=Capsule, length=18.75, radius=5.5, center=<5 0 4>, rotation=<rot 0 0 90>}]
 dyn_shorts_bk_r_MSTR [{shapeType=Capsule, length=1, radius=1}]
 dyn_shorts_ft_r [{shapeType=Capsule, length=16.125, radius=6.125, center=<10 0 3>, rotation=<rot 0 0 90>}]
 dyn_shorts_ft_r_MSTR [{shapeType=Capsule, length=1, radius=1}]
 dyn_shorts_in_r [{shapeType=Capsule, length=17.5, radius=4.75, center=<10 0 3>, rotation=<rot 0 0 90>}]
 dyn_shorts_in_r_MSTR [{shapeType=Capsule, length=1, radius=1}]
 dyn_shorts_out_r [{shapeType=Capsule, length=18, radius=4.25, center=<10 0 3>, rotation=<rot 0 0 90>}]
 dyn_shorts_out_r_MSTR [{shapeType=Capsule, length=1, radius=1}]
 thigh_twist_01_r [{shapeType=Capsule, length=17, radius=6.75, center=<2.375 -0.1263 -0.1081>, rotation=<rot 88.4195 0.1455 0>}]
}
constraints [
 {bone1Name=dyn_shorts_in_r, bone2Name=dyn_shorts_in_r_MSTR, swing1LimitDegrees=7, swing2LimitDegrees=15, twistLimitDegrees=2}
 {bone1Name=dyn_shorts_bk_01_r, bone2Name=dyn_shorts_bk_r_MSTR, swing1LimitDegrees=5, swing2LimitDegrees=15, twistLimitDegrees=5}
 {bone1Name=dyn_shorts_bk_02_r, bone2Name=dyn_shorts_bk_01_r, swing1LimitDegrees=5, swing2LimitDegrees=15, twistLimitDegrees=5}
 {bone1Name=dyn_shorts_out_r, bone2Name=dyn_shorts_out_r_MSTR, swing1LimitDegrees=7, swing2LimitDegrees=15, twistLimitDegrees=2}
 {bone1Name=dyn_shorts_ft_r, bone2Name=dyn_shorts_ft_r_MSTR, swing1LimitDegrees=7, swing2LimitDegrees=15, twistLimitDegrees=2}
 {
  bone1Name=dyn_shorts_ft_r_MSTR
  bone2Name=thigh_twist_01_r
  swing1LimitDegrees=45
  swing1Motion=Free
  swing2LimitDegrees=45
  swing2Motion=Free
  twistLimitDegrees=45
  twistMotion=Free
 }
 {
  bone1Name=dyn_shorts_in_r_MSTR
  bone2Name=thigh_twist_01_r
  swing1LimitDegrees=45
  swing1Motion=Free
  swing2LimitDegrees=45
  swing2Motion=Free
  twistLimitDegrees=45
  twistMotion=Free
 }
 {
  bone1Name=dyn_shorts_bk_r_MSTR
  bone2Name=thigh_twist_01_r
  swing1LimitDegrees=45
  swing1Motion=Free
  swing2LimitDegrees=45
  swing2Motion=Free
  twistLimitDegrees=45
  twistMotion=Free
 }
 {
  bone1Name=dyn_shorts_out_r_MSTR
  bone2Name=thigh_twist_01_r
  swing1LimitDegrees=45
  swing1Motion=Free
  swing2LimitDegrees=45
  swing2Motion=Free
  twistLimitDegrees=45
  twistMotion=Free
 }
]
